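# Master material for GroveGlimmer PetA that animates non-skeletal meshes, including instanced and Nanite ones, from an animation texture via World Position Offset.
# Time, offset per instance or by object position and scaled by a shared PlaybackSpeed, scrolls the sampling U.
# MF_WPOTransform_subuvs outputs offset and normal; the offset passes through a vertex-color hide function, the normal blends with a normal map.
# Static switches swap base color for weight previews; ribbon-LOD branches appear unconnected.
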Material /Hera/Props/GroveGlimmer/GroveGlimmer_PetA/Materials/M_characterWPO_Master_FlipbookEyes
graph {
 properties {
  bHasPixelAnimation=true
  bTangentSpaceNormal=false
  bUseMaterialAttributes=true
  bUsedWithInstancedStaticMeshes=true
  bUsedWithNanite=true
  bUsedWithNiagaraMeshParticles=true
  pixelDepthOffsetMode=PDOM_Legacy
  usedShadingModels=Default Lit
 }
 nodes [
  Constant() {r=0.333}
  MaterialFunctionCall!(In) {materialFunction=/Game/Environments/TechArt/ScreenAnimationWPO/MF_3Dto2DScreen}
  n1 = TextureObjectParameter() {
   parameterName=AnimationTexture
   samplerType=SAMPLERTYPE_LinearColor
   texture=/Game/ArtTools/characterWPO/animation/SoldierMannequin_Female_Emote_LetsBegin_CMM
  }
  n2 = MaterialFunctionCall(Float2=TextureProperty(Tex=@n1)) {materialFunction=&BreakOutFloat2Components}
  n3 = Divide(B=NamedRerouteDeclaration(None=@n2.R) {name=textureWidth, nodeColor=<rgba 0.0078 0 1 1>})
  NamedRerouteDeclaration(None=Multiply(A=NamedRerouteDeclaration(None=@n3) {name=texelWidth, nodeColor=<rgba 0 1 0.2266 1>}) {constB=0.5}) {name=halfTexelWidth, nodeColor=<rgba 0.0312 0 1 1>}
  NamedRerouteDeclaration(None=@n1) {name=AnimationTexture, nodeColor=<rgba 0.6172 0 1 1>}
  n4 = MaterialFunctionCall(Float3=ObjectPositionWS()) {materialFunction=&BreakOutFloat3Components}
  n5 = Add(A=@n4.R, B=@n4.R)
  n6 = Frac(None=Divide(A=Add(A=@n5, B=@n4.B)) {constB=10})
  n7 = ScalarParameter() {defaultValue=33.3, parameterName=RandomTimeOffset}
  n8 = Time()
  n9 = ScalarParameter() {parameterName=TimeOffset}
  n10 = StaticSwitchParameter(True=Add(A=Multiply(A=@n6, B=@n7), B=@n8), False=Add(A=Add(A=@n8, B=@n9), B=PerInstanceRandom())) {parameterName=UseRandomOffsetPosition}
  n11 = Divide(B=NamedRerouteUsage()) {constA=30}
  n12 = CollectionParameter() {collection=/Game/Environments/TechArt/Hologram/MPC_Hologram, parameterName=PlaybackSpeed}
  n13 = Frac(None=Multiply(A=Multiply(A=@n10, B=@n11), B=ComponentMask(None=@n12) {r=true}))
  n14 = NamedRerouteUsage()
  NamedRerouteDeclaration(None=Add(A=Multiply(A=@n13, B=@n14), B=NamedRerouteUsage())) {name=u, nodeColor=<rgba 0 1 0.9766 1>}
  NamedRerouteDeclaration(None=Add(A=NamedRerouteUsage(), B=NamedRerouteUsage())) {name=uStart, nodeColor=<rgba 1 0.3516 0 1>}
  n15 = NamedRerouteUsage()
  NamedRerouteDeclaration(None=Subtract(A=Subtract(B=@n15), B=@n15)) {name=uWidth, nodeColor=<rgba 0.9922 1 0 1>}
  NamedRerouteDeclaration(None=Divide(B=NamedRerouteDeclaration(None=@n2.G) {name=textureHeight, nodeColor=<rgba 0.5469 0 1 1>})) {name=texelHeight, nodeColor=<rgba 0 1 0.2266 1>}
  n16 = VertexColor()
  n17 = MaterialFunctionCall(Hue Shift Percentage=VertexColor().R, Texture=Constant3Vector() {constant=<rgba 1 0 0 0>}) {materialFunction=&HueShift}
  n18 = StaticSwitchParameter(True=@n16, False=@n17.Result) {parameterName=PreviewRaw}
  n19 = TextureSampleParameter2D() {parameterName=BaseColor, texture=&T_BlacksRGB}
  n20 = MaterialFunctionCall!(u=NamedRerouteUsage(), bind_u=NamedRerouteUsage(), vStart=Constant(), vHeight=Constant() {r=1}, texelHeight=NamedRerouteUsage(), WorldNormal) {materialFunction=/Game/ArtTools/characterWPO/MF_WPOTransform_subuvs}
  n21 = MaterialFunctionCall(
   BaseNormal=@n20.Normal
   AdditionalNormal=TextureSampleParameter2D() {parameterName=Normals, samplerType=SAMPLERTYPE_Normal, texture=&FlatNormalBlue}.RGB
  ) {materialFunction=&BlendAngleCorrectedNormals}
  Reroute(None=SetMaterialAttributes!(
    MaterialAttributes
    Base Color=StaticSwitchParameter(True=@n18, False=@n19.RGB) {parameterName=PreviewWeights}
    Normal=@n21.Result
    World Position Offset=MaterialFunctionCall!(WPO=@n20.WPO) {materialFunction=/Game/Packages/Fortress_SharedMaterials/Base_Material_Functions/Fortnite_HideElementsWithVertexColor_2022}.Result
   ) {attributeSetTypes [#1, #2, #3]})
  n22 = ScalarParameter() {parameterName=jointIndex}
  n23 = Divide(A=NamedRerouteUsage())
  StaticSwitchParameter(True=SphereMask(A=@n16.R, B=Divide(A=@n22, B=@n23)) {attenuationRadius=0.01}, False=@n17.Result) {parameterName=pickWeight}
  n24 = ScalarParameter() {defaultValue=20, parameterName=Width}
  n25 = MaterialFunctionCall(WidthBase=@n24, WidthTip=@n24) {materialFunction=&SplineThicken}
  StaticSwitchParameter(True=Add(A=@n20.WPO, B=@n25.WorldPositionOffset), False=@n20.WPO) {parameterName=IsRibbonLOD}
  StaticSwitchParameter(True=@n25.Normal, False=@n21.Result) {parameterName=IsRibbonLOD}
 ]
}
